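# Unlit sky material for the SparksLobby Season 07 skybox, drawing a night sky with moon and stars.
# An optional textured moon is projected opposite the atmosphere light using cross-product axes computed per vertex, then edge-masked.
# Two panning star samples multiply into twinkling stars, fading by height over a horizon-to-top color gradient.
# A mobile-only branch derives a vertical offset from camera height.
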Material /SparksLobby/Environment/TimeOfDay/SkyBoxes/M_Sky_Season_07
graph {
 properties {bAutomaticallySetUsageInEditor=false, bCastRayTracedShadows=false, bIsSky=true, shadingModel=MSM_Unlit, usedShadingModels=Unlit}
 nodes [
  n1 = Multiply(A=ComponentMask(None=SkyAtmosphereLightDirection()) {b=true, g=true, r=true}) {constB=-1}
  n2 = CrossProduct(A=@n1, B=Constant3Vector() {constant=<rgba 0 0 -1 0>})
  n3 = CrossProduct(A=@n2, B=@n1)
  n4 = NamedRerouteUsage()
  n5 = AppendVector(A=DotProduct(A=Normalize(VectorInput=@n3), B=@n4), B=DotProduct(A=Normalize(VectorInput=@n2), B=@n4))
  n6 = ScalarParameter() {defaultValue=0.1, group=Sun-Moon, parameterName=MoonSize}
  n7 = VertexInterpolator(VS=Add(A=Divide(A=@n5, B=@n6)) {constB=0.5})
  n8 = TextureSampleParameter2D(UVs=@n7.PS) {
   group=Sun-Moon
   parameterName=MoonTexture
   samplerSource=SSM_Wrap_WorldGroupSettings
   samplerType=SAMPLERTYPE_LinearColor
   texture=/Game/TimeOfDay/Skydome/T_Moon_Artemis
  }
  n9 = Constant() {r=0.5}
  n10 = MaterialFunctionCall(Float2=@n7.PS) {materialFunction=&BreakOutFloat2Components}
  n11 = SphereMask(A=@n9, B=@n10.R) {attenuationRadius=0.5}
  n12 = SphereMask(A=@n10.G, B=@n9) {attenuationRadius=0.5}
  n13 = Multiply(
   A=Multiply(A=@n8.RGB, B=Multiply(A=@n11, B=@n12))
   B=VectorParameter() {group=Sun-Moon, parameterName=MoonColor, defaultValue=<rgba 1.7 1.7 1.7 1>}.RGB
  )
  n14 = ScalarParameter() {defaultValue=4, parameterName=MoonIntensity}
  n15 = Panner(Coordinate=TextureCoordinate() {uTiling=10, vTiling=10}) {speedX=-0.002, speedY=0.001}
  n16 = Panner(Coordinate=TextureCoordinate() {uTiling=7, vTiling=7}) {speedX=0.002, speedY=0.0025}
  n17 = ComponentMask(None=Normalize(VectorInput=WorldPosition() {worldPositionShaderOffset=WPT_ExcludeAllShaderOffsets}.XYZ)) {b=true}
  n18 = Multiply(
   A=Multiply(
    A=TextureSample(UVs=@n15) {texture=&/T_Skybox_Stars}.RGB
    B=TextureSample(UVs=@n16) {texture=&T_Skybox_Stars}.RGB
   ) {constB=100}
   B=@n17
  )
  n19 = ScalarParameter() {defaultValue=10, parameterName=Star_Brightness}
  n20 = VectorParameter() {parameterName=Horizon_Sky, defaultValue=<rgba 0.0188 0.0362 0.0417 1>}
  n21 = VectorParameter() {parameterName=Top_Sky, defaultValue=<rgba 0.0625 0.0462 0.0938 1>}
  n22 = Multiply(A=@n17) {constB=2}
  Add(
   A=StaticSwitchParameter(True=Multiply(A=@n13, B=@n14) {constB=14}, False=Constant()) {defaultValue=true, parameterName=MoonActive}
   B=Add(A=Multiply(A=@n18, B=@n19) {constB=10000}, B=LinearInterpolate(A=@n20.RGB, B=@n21.RGB, Alpha=@n22))
  )
  n23 = WorldPosition()
  n24 = CameraPositionWS()
  NamedRerouteDeclaration(None=Normalize(VectorInput=Subtract(A=@n23.XYZ, B=@n24))) {name=CameraVectorForVertexShader, nodeColor=<rgba 0 1 0.0859 1>}
  n25 = ComponentMask(None=CameraPositionWS()) {b=true}
  ShadingPathSwitch(Default=Constant(), Mobile=AppendVector(A=Constant2Vector(), B=Multiply(A=@n25) {constB=0.9}))
 ]
}
