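# Unlit sky material for the Season 08 sunset skybox in the Sparks lobby.
# Sky Atmosphere luminance, sampled along a mirrored camera vector, is summed with stars, then passed through halo and high-altitude cloud functions.
# A vertical two-color UV gradient is added on top.
# A named Sun Bloom term raises the saturated, negated dot of light direction and mirrored view vector to a power, tinted by a color parameter defaulting to black.
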Material /SparksLobby/Environment/TimeOfDay/SkyBoxes/M_Sky_Season_08_Sunset
graph {
 properties {bAutomaticallySetUsageInEditor=false, bIsSky=true, shadingModel=MSM_Unlit, usedShadingModels=Unlit}
 nodes [
  n1 = VectorParameter() {parameterName=Color3, defaultValue=<rgba 0.6771 0.2643 0.2059 1>}
  n2 = VectorParameter() {parameterName=Color4, defaultValue=<rgba 0.0104 0.0041 0.0041 1>}
  n3 = ComponentMask(None=TextureCoordinate()) {g=true}
  n4 = MaterialFunctionCall!(Multiplier=Constant() {r=-1}, Offset) {materialFunction=/Game/TimeOfDay/TODM_Sequence/Materials/Sky/MF_MirroredCameraVector}
  n5 = NamedRerouteUsage()
  n6 = MaterialFunctionCall!(
   Sky Atmosphere=Add(
    A=SkyAtmosphereViewLuminance!(WorldDirection=@n4.Result)
    B=MaterialFunctionCall!(HorizonMask, UV) {materialFunction=/Game/TimeOfDay/TODM_Sequence/Materials/Sky/MaterialFunctions/MF_Sky_Stars}.Result
   )
   Horizon Mask=@n5
  ) {materialFunction=/Game/TimeOfDay/TODM_Sequence/Materials/Sky/MaterialFunctions/MF_Sky_Halo}
  n7 = MaterialFunctionCall!(Color=@n6.Result, Halo Mask=@n6.Result, Horizon Mask=@n5) {materialFunction=/Game/TimeOfDay/TODM_Sequence/Materials/Sky/MaterialFunctions/MF_Sky_HighAltitudeClouds}
  Add(
   A=NamedRerouteUsage()
   B=Add(
    A=LinearInterpolate(A=@n1.RGB, B=@n2.RGB, Alpha=@n3)
    B=NamedRerouteDeclaration(None=@n7.Color) {name=Pre Skylight Captures, nodeColor=<rgba 0 1 0.5781 1>}
   )
  )
  n8 = SkyAtmosphereLightDirection()
  n9 = MaterialFunctionCall() {materialFunction=&MF_MirroredCameraVector}
  n10 = Saturate(None=Multiply(A=DotProduct(A=@n8, B=@n9.Result)) {constB=-1})
  n11 = ScalarParameter() {defaultValue=22.8659, parameterName=SunBloomExp}
  NamedRerouteDeclaration(None=Multiply(A=Power(Base=@n10, Exp=@n11), B=VectorParameter() {parameterName=SunBloomColor, defaultValue=<rgba 0 0 0 1>}.RGB)) {name=Sun Bloom, nodeColor=<rgba 1 0 0.8906 1>}
 ]
}
