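# Additive unlit material that draws the sun, moon and halo of the time-of-day sky.
# Radial UV gradients form a quality-switched sun disc (soft by default, hard-edged on Epic), a textured moon and a halo, each color-tinted.
# A camera-vector mask hides everything below the horizon, and reflection captures receive black.
# Vertex offset math rebuilds the mesh as a distant, sized quad facing along the SkyAtmosphere light direction, shifted by Horizon Offset.
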Material /SparksLobby/Environment/TimeOfDay/SkyBoxes/M_SundAndMoon_S08
graph {
 properties {
  bApplyCloudFogging=true
  bAutomaticallySetUsageInEditor=false
  bComputeFogPerPixel=true
  blendMode=BLEND_Additive
  opacityMaskClipValue=0.5
  shadingModel=MSM_Unlit
  translucencyPass=MTP_BeforeDOF
  usedShadingModels=Unlit
 }
 nodes [
  n1 = Multiply(A=VectorParameter() {parameterName=Sun Color, defaultValue=<rgba 1 0.2482 0 1>}.RGB) {constB=5000}
  n2 = Add(A=TextureCoordinate()) {constB=-0.5}
  n3 = Distance(A=Constant(), B=@n2)
  n4 = OneMinus(None=Saturate(None=Multiply(A=@n3) {constB=2}))
  n5 = Saturate(None=Add(A=@n4, B=ScalarParameter() {defaultValue=-0.64, parameterName=SunMask}))
  n6 = Multiply(A=@n5, B=@n5)
  n7 = Multiply(A=Multiply(A=@n6, B=@n6)) {constB=50}
  n8 = Floor(None=Add(A=@n4) {constB=0.333})
  n9 = TextureSampleParameter2D(UVs=Add(A=Multiply(A=@n2) {constB=3}) {constB=0.5}) {
   group=Sun-Moon
   parameterName=MoonTexture
   samplerSource=SSM_Clamp_WorldGroupSettings
   samplerType=SAMPLERTYPE_LinearColor
   texture=/Game/TimeOfDay/Skydome/T_Moon_Artemis
  }
  n10 = Multiply(A=Multiply(A=@n8, B=@n9.R), B=Multiply(A=VectorParameter() {parameterName=Moon Color, defaultValue=<rgba 0 0 0 1>}.RGB) {constB=10})
  n11 = Add(A=QualitySwitch(Default=Multiply(A=@n1, B=@n7), Epic=Multiply(A=@n1, B=@n8)), B=Multiply(A=@n10, B=@n10))
  n12 = Multiply(A=@n4, B=@n4)
  n13 = Multiply(A=Multiply(A=@n12, B=@n12), B=VectorParameter() {parameterName=Halo Color, defaultValue=<rgba 10 4.4549 0 1>}.RGB)
  n14 = CameraVectorWS()
  n15 = Saturate(None=Multiply(A=ComponentMask(None=@n14) {b=true}) {constB=-1})
  ReflectionCapturePassSwitch!(Default=Multiply(A=Add(A=@n11, B=@n13), B=Ceil(None=@n15)), Reflection=Constant())
  SkyAtmosphereLightDirection()
  n16 = LocalPosition()
  n17 = Multiply(A=Multiply(
    A=Transform(None=@n16.XYZ)
    B=MaterialFunctionCall() {materialFunction=&ObjectScale}.Scale XYZ
   )) {constB=-1}
  n18 = Add(A=SkyAtmosphereLightDirection(), B=AppendVector(A=Constant2Vector(), B=ScalarParameter() {parameterName=Horizon Offset}))
  n19 = Multiply(A=@n18) {constB=3000000}
  n20 = LocalPosition()
  n21 = Normalize(VectorInput=CrossProduct(A=Constant3Vector() {constant=<rgba 0 0 1 0>}, B=@n18))
  n22 = Add(
   A=Multiply(A=ComponentMask(None=@n20.XY) {g=true}, B=@n21)
   B=Multiply(A=ComponentMask(None=@n20.XY) {r=true}, B=CrossProduct(A=@n21, B=@n18))
  )
  n23 = ScalarParameter() {defaultValue=3500, parameterName=Size}
  Subtract(A=Add(A=Add(A=@n17, B=@n19), B=Multiply(A=@n22, B=@n23) {constB=60}), B=ActorPositionWS() {originType=CameraRelative})
 ]
}
